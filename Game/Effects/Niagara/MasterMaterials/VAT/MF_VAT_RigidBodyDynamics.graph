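# Material function that plays SideFX Houdini Vertex Animation Textures in rigid-body dynamics mode, animating mesh pieces through World Position Offset.
# Frame count, texture format and scale limits are decoded from fractional mesh bounds, or from legacy parameters for instancing.
# Auto-playback time or Display Frame drives position, rotation and color texture sampling.
# Pieces rotate and scale around UV-encoded pivots, with optional quaternion slerp, acceleration-smoothed trajectories and a debug plane.
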MaterialFunction /Game/Effects/Niagara/MasterMaterials/VAT/MF_VAT_RigidBodyDynamics
graph {
 properties {
  bExposeToLibrary=true
  description="Supports SideFX Labs Vertex Animation Textures Rigid-Body Dynamics (Rigid) mode. On the Material's Details panel, turn off Tangent Space Normal and set Num Customized UVs to 5 (or 6 to support ISM/HISM). Surface texture UVs are accessible at TexCoord[0]. (All native parameters are inside this Material Function. Create a Material Instance to gain access to the parameters. Avoid creating your own parameters outside this Material Function with names that clash with the parameter names inside this Material Function.)"
  libraryCategoriesText [SideFX]
 }
 nodes [
  n1 = ScalarParameter() {
   desc="To set this parameter, first close the Material Instace. In Content Browser, select the Material Instance and the Data Table before right-clicking on either and then selecting Scripted Actions (or Scripted Asset Actions in UE5) > SideFX Legacy Set VAT Material Instance from Data Table. You need to have exported the Real-Time Data JSON file from Houdini and imported it to Unreal as DataTable (Row Type: SideFX_VAT_Material_Parameters). Alternatively, you can manually copy and paste the value from the JSON file."
   group=Houdini VAT - Instancing
   parameterName=Bound Min X
   sortPriority=4
  }
  n2 = ScalarParameter() {
   desc="To set this parameter, first close the Material Instace. In Content Browser, select the Material Instance and the Data Table before right-clicking on either and then selecting Scripted Actions (or Scripted Asset Actions in UE5) > SideFX Legacy Set VAT Material Instance from Data Table. You need to have exported the Real-Time Data JSON file from Houdini and imported it to Unreal as DataTable (Row Type: SideFX_VAT_Material_Parameters). Alternatively, you can manually copy and paste the value from the JSON file."
   group=Houdini VAT - Instancing
   parameterName=Bound Min Y
   sortPriority=5
  }
  n3 = ScalarParameter() {
   desc="To set this parameter, first close the Material Instace. In Content Browser, select the Material Instance and the Data Table before right-clicking on either and then selecting Scripted Actions (or Scripted Asset Actions in UE5) > SideFX Legacy Set VAT Material Instance from Data Table. You need to have exported the Real-Time Data JSON file from Houdini and imported it to Unreal as DataTable (Row Type: SideFX_VAT_Material_Parameters). Alternatively, you can manually copy and paste the value from the JSON file."
   group=Houdini VAT - Instancing
   parameterName=Bound Min Z
   sortPriority=6
  }
  n4 = ScalarParameter() {
   defaultValue=0.0001
   desc="This value should always be something like: 0.01, 0.001, 0.0001, etc. It controls to what decimal place to round the embedded data. The default value should work in most cases, but if there are visual glitches, try removing one or two zeros to increase this value if your animation spans a large area, or adding one or two zeros to decrease this value if your animation is confined to a small area.\r\n\r\nWithout rounding, the calculations based on the embedded data may be susceptible to floating point errors in Unreal Engine."
   group=Houdini VAT - Debug
   parameterName=Embedded Data Rounding
   sortPriority=4
  }
  n5 = Divide(B=@n4)
  n6 = Multiply(A=Custom() {code="GetPrimitiveData(Parameters.PrimitiveId).LocalObjectBoundsMin.xyz", description=Local Bounds Min, inputs []}, B=@n5) {constB=10000}
  n7 = Multiply(A=@n4) {constB=0.1}
  n8 = StaticSwitchParameter(
   True=Multiply(A=MaterialFunctionCall(X=@n1, Y=@n2, Z=@n3) {materialFunction=&MakeFloat3}.Result) {constB=0.1}
   False=Multiply(A=Round(None=@n6), B=@n7) {constB=0}
  ) {
   desc="If you want to use this Material Instance with ISM/HISM or mesh particles, turn this on and turn on Support Real-Time Instancing in Houdini.\r\n\r\nIf you simply want to use the legacy parameters or modify the object's bounds, turn this on and only turn on Allow Exporting Real-Time Data JSON File (Legacy) in Houdini without turning on Support Real-Time Instancing.\r\n\r\nLegacy parameters are a list of numerical values exported through a JSON file. They contain the embedded data just like the actual mesh, but when Support Legacy Parameters and Instancing is turned on, the shader will read the bounds and the embedded data from the legacy parameters instead of the actual mesh.\r\n\r\nUsing legacy parameters is less convenient, but it does produce more accurate results if the animation spans a huge area; it also leads to lower instruction counts."
   group=Houdini VAT - Instancing
   parameterName=Support Legacy Parameters and Instancing
   sortPriority=0
  }
  n9 = ComponentMask(None=@n8) {g=true}
  n10 = OneMinus(None=Subtract(A=Ceil(None=@n9), B=@n9)) {desc=Active Pixels Ratio X}
  n11 = TextureCoordinate() {coordinateIndex=1}
  n12 = Multiply(A=@n10, B=ComponentMask(None=@n11) {r=true}) {constA=1}
  n13 = ComponentMask(None=@n8) {r=true}
  n14 = OneMinus(None=Subtract(A=Ceil(None=@n13), B=@n13)) {desc=Active Pixels Ratio Y}
  n15 = Multiply(A=ComponentMask(None=@n11) {g=true}, B=@n14)
  n16 = FunctionInput(Preview=Time()) {
   bUsePreviewValueAsDefault=true
   description="The animation will use this input when the Material Instance parameter Auto Play is turned on. Otherwise, the animation uses the parameter Display Frame."
   inputName="Time Override (VS)"
   inputType=FunctionInput_Scalar
   sortPriority=-2
  }
  n17 = ScalarParameter() {
   desc="Used when Auto Play is turned on, this helps you start the animation from its first frame when a specific gameplay event is triggered.\r\n\r\nUse Blueprints or C++ code to set this parameter through a Material Instance Dynamic. The value should be set to the output of the Blueprint node Get Time Seconds (or the C++ function UWorld::GetTimeSeconds) the moment this event is triggered. It should only be set once.\r\n\r\nThis method is compatible with Motion Blur, whereas setting Display Frame is not. To fully enable Motion Blur, turn on Output Velocities during Base Pass in Project Settings."
   group=Houdini VAT
   parameterName=Game Time at First Frame
   sortPriority=-9
  }
  n18 = ScalarParameter() {
   defaultValue=60
   desc="The FPS of the Houdini HIP file when exporting the animation. It is possible for you to simulate at a higher FPS, cache the data, and then export at a lower FPS. This parameter refers to the export FPS. Make sure this parameter is set correctly whenever it is exposed on the Material Instance interface."
   group=Houdini VAT
   parameterName=Houdini FPS
   sortPriority=-6
  }
  n19 = TextureObjectParameter() {
   desc="The RGB channels store positions. The A channel usually stores piece scales, but it can also store custom float data."
   group=Houdini VAT
   parameterName=Position Texture
   samplerType=SAMPLERTYPE_LinearColor
   sortPriority=10
   texture=/Engine/Functions/Engine_MaterialFunctions02/PivotPainter2/Black_1x1_EXR_Texture
  }
  n20 = ComponentMask(None=TextureProperty(Tex=@n19)) {g=true}
  n21 = ComponentMask(None=@n8) {b=true}
  n22 = Multiply(A=Subtract(A=Ceil(None=@n21), B=@n21)) {constB=100}
  n23 = ScalarParameter() {
   desc="To set this parameter, first close the Material Instace. In Content Browser, select the Material Instance and the Data Table before right-clicking on either and then selecting Scripted Actions (or Scripted Asset Actions in UE5) > SideFX Legacy Set VAT Material Instance from Data Table. You need to have exported the Real-Time Data JSON file from Houdini and imported it to Unreal as DataTable (Row Type: SideFX_VAT_Material_Parameters). Alternatively, you can manually copy and paste the value from the JSON file."
   group=Houdini VAT - Instancing
   parameterName=Bound Max X
   sortPriority=1
  }
  n24 = ScalarParameter() {
   desc="To set this parameter, first close the Material Instace. In Content Browser, select the Material Instance and the Data Table before right-clicking on either and then selecting Scripted Actions (or Scripted Asset Actions in UE5) > SideFX Legacy Set VAT Material Instance from Data Table. You need to have exported the Real-Time Data JSON file from Houdini and imported it to Unreal as DataTable (Row Type: SideFX_VAT_Material_Parameters). Alternatively, you can manually copy and paste the value from the JSON file."
   group=Houdini VAT - Instancing
   parameterName=Bound Max Y
   sortPriority=2
  }
  n25 = ScalarParameter() {
   desc="To set this parameter, first close the Material Instace. In Content Browser, select the Material Instance and the Data Table before right-clicking on either and then selecting Scripted Actions (or Scripted Asset Actions in UE5) > SideFX Legacy Set VAT Material Instance from Data Table. You need to have exported the Real-Time Data JSON file from Houdini and imported it to Unreal as DataTable (Row Type: SideFX_VAT_Material_Parameters). Alternatively, you can manually copy and paste the value from the JSON file."
   group=Houdini VAT - Instancing
   parameterName=Bound Max Z
   sortPriority=3
  }
  n26 = Multiply(A=Custom() {code="GetPrimitiveData(Parameters.PrimitiveId).LocalObjectBoundsMax.xyz", description=Local Bounds Max, inputs []}, B=@n5) {constB=10000}
  n27 = StaticSwitchParameter(
   True=Multiply(A=MaterialFunctionCall(X=@n23, Y=@n24, Z=@n25) {materialFunction=&MakeFloat3}.Result) {constB=0.1}
   False=Multiply(A=Round(None=@n26), B=@n7) {constB=0}
  ) {
   desc="If you want to use this Material Instance with ISM/HISM or mesh particles, turn this on and turn on Support Real-Time Instancing in Houdini.\r\n\r\nIf you simply want to use the legacy parameters or modify the object's bounds, turn this on and only turn on Allow Exporting Real-Time Data JSON File (Legacy) in Houdini without turning on Support Real-Time Instancing.\r\n\r\nLegacy parameters are a list of numerical values exported through a JSON file. They contain the embedded data just like the actual mesh, but when Support Legacy Parameters and Instancing is turned on, the shader will read the bounds and the embedded data from the legacy parameters instead of the actual mesh.\r\n\r\nUsing legacy parameters is less convenient, but it does produce more accurate results if the animation spans a huge area; it also leads to lower instruction counts."
   group=Houdini VAT - Instancing
   parameterName=Support Legacy Parameters and Instancing
   sortPriority=0
  }
  n28 = ComponentMask(None=@n27) {g=true}
  n29 = Subtract(A=@n28, B=Floor(None=@n28)) {desc="[0.5, 1.0) if texture data are in HDR.\r\n[0.0, 0.5) if texture data are in LDR."}
  n30 = Round(None=Multiply(A=@n29) {constB=100}) {desc="[0, 99]"}
  n31 = Constant() {r=25}
  n32 = Constant() {desc=Near-zero, r=0.05}
  n33 = Constant()
  n34 = Constant() {r=25}
  n35 = Fmod(A=@n30, B=Constant() {r=50})
  n36 = Subtract(A=@n34, B=Abs(None=Subtract(A=@n35, B=@n34) {constB=25}))
  n37 = Subtract(A=@n36, B=Floor(None=Divide(A=@n36) {constB=10}))
  n38 = If(A=Fmod(A=@n30, B=Constant() {r=10}), B=@n32, "A > B"=@n37, "A == B"=@n37, "A < B"=@n33)
  n39 = Round(None=If(A=Fmod(A=@n30, B=@n31), B=@n32, "A > B"=@n38, "A == B"=@n38, "A < B"=@n33))
  n40 = Divide(A=Round(None=Multiply(A=@n20, B=@n14)), B=Add(A=Round(None=@n22), B=Multiply(A=@n39) {constB=100}) {desc=Animation UV row count}) {desc=Animation frame count}
  n41 = Multiply(A=Subtract(A=@n16, B=@n17), B=Divide(A=@n18, B=Subtract(A=@n40) {constB=0.01}))
  n42 = ScalarParameter() {
   defaultValue=1
   desc="Use this to scale playback speed. When this is 1.0 and Original FPS is correct, the animation will be playing at its original speed."
   group=Houdini VAT
   parameterName=Playback Speed
   sortPriority=-7
  }
  n43 = Multiply(A=Frac(None=Multiply(A=@n41, B=@n42)), B=@n40)
  n44 = ScalarParameter() {
   defaultValue=1
   desc="The exact animation frame to display, unaffected by Playback Speed or Houdini FPS. When this is set to 1.0, it displays the first frame of the animation. This frame sequence exactly matches the frame sequence in Houdini. If the Interframe Interpolation parameter is available and turned on, fractional values in Display Frame will generate interpolated results between adjacent frames.\r\n\r\nUse Blueprints or C++ code to drive this parameter through a Material Instance Dynamic.\r\n\r\nThis method is incompatible with Motion Blur. To support Motion Blur, turn on Auto Play and follow the instructions in the tooltip of Game Time at First Frame."
   group=Houdini VAT
   parameterName=Display Frame
   sliderMax=1000
   sliderMin=1
   sortPriority=-8
  }
  n45 = FunctionInput(Preview=@n44) {
   bUsePreviewValueAsDefault=true
   description=Overrides the Material Instance parameter Display Frame. The animation will ignore this input when the Material Instance parameter Auto Play is turned on.
   inputName="Display Frame Override (VS)"
   inputType=FunctionInput_Scalar
   sortPriority=-1
  }
  n46 = StaticSwitchParameter(True=Add(A=Floor(None=@n43)), False=Floor(None=@n45)) {
   defaultValue=true
   desc="Automatically plays back the animation in loops. The animation speed is affected by both Playback Speed and Houdini FPS.\r\n\r\nFor this mode, turn this on if you want to support Motion Blur."
   group=Houdini VAT
   parameterName=Auto Playback
   sortPriority=-10
  }
  n47 = Divide(B=@n40)
  n48 = Multiply(A=Fmod(A=@n46, B=@n40), B=@n47)
  n49 = Add(A=@n15, B=Multiply(A=@n48, B=@n14))
  n50 = AppendVector(A=@n12, B=@n49) {desc=Next frame UV}
  n51 = TextureObjectParameter() {
   desc="The data can come from the \"Cd\" and \"Alpha\" attributes, or custom vector and float attributes."
   group=Houdini VAT
   parameterName=Color Texture
   samplerType=SAMPLERTYPE_LinearColor
   sortPriority=13
   texture=&Black_1x1_EXR_Texture
  }
  n52 = TextureSample(UVs=@n50, Tex=@n51) {texture=&Black}
  FunctionOutput(None=@n52.RGBA) {
   description="Outputs next frame's Color Texture RGBA on the vertex shader (VS). The data can come from the \"Cd\" and \"Alpha\" attributes, or custom vector and float attributes. To get interpolated data, use this with Color RGBA This Frame (VS) and Interframe Interpolation Alpha (VS) to create your own Lerp."
   outputName="Color RGBA Next Frame (VS)"
   sortPriority=12
  }
  n53 = TextureObjectParameter() {
   desc="The data strictly come from the \"Cd2\" and \"Alpha2\" attributes. You can use Spare Color Texture to store the actual colors and Color Texture to store custom attributes."
   group=Houdini VAT
   parameterName=Spare Color Texture
   samplerType=SAMPLERTYPE_LinearColor
   sortPriority=14
   texture=&Black_1x1_EXR_Texture
  }
  n54 = TextureSample(UVs=@n50, Tex=@n53) {texture=&Black}
  FunctionOutput(None=@n54.RGBA) {
   description="Outputs next frame's Spare Color Texture RGBA on the vertex shader (VS). The data strictly come from the \"Cd2\" and \"Alpha2\" attributes. You can use Spare Color Texture to store the actual colors and Color Texture to store custom attributes. To get interpolated data, use this with Spare Color RGBA This Frame (VS) and Interframe Interpolation Alpha (VS) to create your own Lerp."
   outputName="Spare Color RGBA Next Frame (VS)"
   sortPriority=14
  }
  n55 = TextureSample(UVs=@n50, Tex=@n19) {texture=&Black}
  n56 = TextureObjectParameter() {
   desc=The RGB channels store positions. See the tooltip of Positions Require Two Textures.
   group=Houdini VAT
   parameterName=Position Texture 2
   samplerType=SAMPLERTYPE_LinearColor
   sortPriority=11
   texture=&Black_1x1_EXR_Texture
  }
  n57 = TextureSample(UVs=@n50, Tex=@n56) {texture=&Black}
  n58 = StaticSwitchParameter(True=Add(A=@n55.RGB, B=Multiply(A=@n57.RGB) {constB=0.01}), False=@n55.RGB) {
   desc="This setting's value must match that of Split Positions into Two Textures in Houdini. When the Houdini counterpart is turned on, the position data are split into two parts and encoded in the two position textures, so you have to use them both."
   group=Houdini VAT
   parameterName=Positions Require Two Textures
   sortPriority=5
  }
  n59 = Multiply(A=@n8) {constB=10}
  n60 = Subtract(A=Multiply(A=@n27) {constB=10}, B=@n59)
  n61 = If(A=@n29, B=Constant() {r=0.5}, "A > B"=@n58, "A == B"=@n58, "A < B"=Add(A=Multiply(A=@n58, B=@n60) {constB=20}, B=@n59)) {constB=0.5, desc=Next frame Position}
  FunctionOutput(None=@n61) {
   description="Outputs next frame's piece local position on the vertex shader (VS). If Position Texture is non-HDR, this output is Position Texture RGB transformed by the object local bounds. To get interpolated data, use this with the outputs Piece Local Position This Frame (VS) and Interframe Interpolation Alpha (VS) to create your own Lerp."
   outputName="Piece Local Position Next Frame (VS)"
   sortPriority=8
  }
  n62 = Fmod(A=Floor(None=@n46) {desc=Integer frame}, B=@n40)
  n63 = Saturate(None=Abs(None=Subtract(A=@n62)))
  n64 = Saturate(None=Abs(None=Subtract(A=@n62, B=@n40)))
  n65 = Multiply(A=Fmod(A=Subtract(A=@n46) {constB=2}, B=@n40), B=@n47)
  n66 = AppendVector(A=@n12, B=Add(A=@n15, B=Multiply(A=@n65, B=@n14))) {desc=Last frame UV}
  n67 = TextureSample(UVs=@n66, Tex=@n19) {texture=&Black}
  n68 = TextureSample(UVs=@n66, Tex=@n56) {texture=&Black}
  n69 = StaticSwitchParameter(True=Add(A=@n67.RGB, B=Multiply(A=@n68.RGB) {constB=0.01}), False=@n67.RGB) {
   desc="This setting's value must match that of Split Positions into Two Textures in Houdini. When the Houdini counterpart is turned on, the position data are split into two parts and encoded in the two position textures, so you have to use them both."
   group=Houdini VAT
   parameterName=Positions Require Two Textures
   sortPriority=5
  }
  n70 = If(A=@n29, B=Constant() {r=0.5}, "A > B"=@n69, "A == B"=@n69, "A < B"=Add(A=Multiply(A=@n69, B=@n60) {constB=20}, B=@n59)) {constB=0.5, desc=Last frame Position}
  MaterialExpressionMultiply_57 = Multiply(A=Multiply(A=Subtract(A=@n61, B=@n70)) {constB=0.5}, B=@n18) {desc=Central-difference velocity}
  MaterialExpressionFunctionInput_1 = FunctionInput() {
   description=Use this input only when the Material Instance parameter Smoothly Interpolate Trajectories is on.
   inputName="Local-Space Piece Acceleration This Frame (VS)"
   sortPriority=3
  }
  n71 = Multiply(A=Fmod(A=Subtract(A=@n46), B=@n40), B=@n47)
  n72 = Add(A=@n15, B=Multiply(A=@n71, B=@n14))
  n73 = AppendVector(A=@n12, B=@n72) {desc=This frame UV}
  n74 = TextureSample(UVs=@n73, Tex=@n19) {texture=&Black}
  n75 = TextureSample(UVs=@n73, Tex=@n56) {texture=&Black}
  n76 = StaticSwitchParameter(True=Add(A=@n74.RGB, B=Multiply(A=@n75.RGB) {constB=0.01}), False=@n74.RGB) {
   desc="This setting's value must match that of Split Positions into Two Textures in Houdini. When the Houdini counterpart is turned on, the position data are split into two parts and encoded in the two position textures, so you have to use them both."
   group=Houdini VAT
   parameterName=Positions Require Two Textures
   sortPriority=5
  }
  MaterialExpressionIf_7 = If(A=@n29, B=Constant() {r=0.5}, "A > B"=@n76, "A == B"=@n76, "A < B"=Add(A=Multiply(A=@n76, B=@n60) {constB=20}, B=@n59)) {constB=0.5, desc=This frame Position}
  n77 = StaticSwitchParameter(True=@n43, False=@n45) {
   defaultValue=true
   desc="Automatically plays back the animation in loops. The animation speed is affected by both Playback Speed and Houdini FPS.\r\n\r\nFor this mode, turn this on if you want to support Motion Blur."
   group=Houdini VAT
   parameterName=Auto Playback
   sortPriority=-10
  }
  n78 = Frac(None=@n77) {desc=Interpolation alpha}
  MaterialExpressionDivide_2 = Divide(A=@n78, B=@n18) {desc="Alpha × TimeInc"}
  n79 = Custom!(V=@MaterialExpressionMultiply_57, A=@MaterialExpressionFunctionInput_1, P=@MaterialExpressionIf_7, T=@MaterialExpressionDivide_2) {
   code=V * T + 0.5 * A * T * T + P
   description=Interframe Position
   inputs [
    {inputName=V, input {expression=&self:MaterialExpressionMultiply_57}}
    {inputName=A, input {expression=&self:MaterialExpressionFunctionInput_1}}
    {inputName=P, input {expression=&self:MaterialExpressionIf_7}}
    {inputName=T, input {expression=&self:MaterialExpressionDivide_2}}
   ]
  }
  n80 = LinearInterpolate(A=@MaterialExpressionIf_7, B=@n61, Alpha=@n78)
  n81 = StaticSwitchParameter(
   True=If(A=Multiply(A=@n63, B=@n64) {desc=Outputs 0 on the first and last frames}, B=Constant(), "A > B"=@n79, "A == B"=@n80, "A < B"=@n80)
   False=@n80
  ) {
   desc="This is an expensive feature, but it will produce much smoother interpolated trajectories when the pieces are moving along high-curvature paths. To use this feature, Support Smoothly Interpolated Trajectories must be turned on in Houdini and Color Texture must be exported.\r\n\r\nYou can then access the acceleration data from this Material Function's Color RGBA This Frame (VS) output. Convert them from Houdini's coordinate system to Unreal's coordinate system using MF_TransformVelocity and feed the result to this Material Function's Local-Space Piece Acceleration This Frame (VS) input."
   group=Houdini VAT - Customization
   parameterName=Smoothly Interpolated Trajectories
   sortPriority=3
  }
  n82 = StaticSwitchParameter(True=@n81, False=@MaterialExpressionIf_7) {
   desc=Interpolates interframe data when the animation frame is a fractional number. This results in smooth visuals even when you slow down the animation or when the frame rate is unstable.
   group=Houdini VAT
   parameterName=Interframe Interpolation
   sortPriority=1
  }
  FunctionOutput(None=@n82) {
   description="Outputs the piece local position on the vertex shader (VS). This output takes into account the current values of both the Material Instance parameters Interframe Interpolation and Smoothly Interpolated Trajectories, but it ignores this Material Function's input Additional Local Position Offset (VS), as that input can vary per vertex."
   outputName="Piece Local Position Final (VS)"
   sortPriority=19
  }
  n83 = ComponentMask(None=@n27) {b=true}
  n84 = Floor(None=@n83)
  n85 = Divide(B=OneMinus(None=Subtract(A=@n83, B=@n84))) {desc=Max Pscale}
  n86 = ComponentMask(None=@n27) {r=true}
  n87 = Floor(None=@n86)
  n88 = Divide(B=OneMinus(None=Subtract(A=@n86, B=@n87))) {desc="Max RPF (revolutions per frame)"}
  n89 = MaterialFunctionCall(X=@n29, Y=@n85, Z=@n88) {materialFunction=&MakeFloat3}
  n90 = Constant() {r=8}
  n91 = MaterialFunctionCall(X=@n10, Y=@n14, Z=Round(None=@n40)) {materialFunction=&MakeFloat3}
  n92 = Add(A=TextureCoordinate(), B=Constant2Vector() {g=-0.05})
  n93 = TextureSample(UVs=@n73, Tex=@n51) {texture=&Black}
  n94 = StaticSwitchParameter(True=LinearInterpolate(A=@n93.RGBA, B=@n52.RGBA, Alpha=@n78), False=@n93.RGBA) {
   desc="There is a cost associated with interpolating each vertex animation texture. While Interframe Interpolation is the overall switch, you can independently decide whether or not to interpolate Color Texture. This only affects the pixel shader outputs Color RGB (PS) and Color Alpha (PS). The outputs ending with This Frame (VS) / Next Frame (VS) are never interpolated."
   group=Houdini VAT
   parameterName=Interpolate Color
   sortPriority=2
  }
  n95 = StaticSwitchParameter(True=@n94, False=@n93.RGBA) {
   desc=Interpolates interframe data when the animation frame is a fractional number. This results in smooth visuals even when you slow down the animation or when the frame rate is unstable.
   group=Houdini VAT
   parameterName=Interframe Interpolation
   sortPriority=2
  }
  n96 = VertexInterpolator(VS=@n95)
  n97 = StaticSwitchParameter(
   True=Add(
    A=MaterialFunctionCall!(Vector3=@n89.Result, MaximumNumberOfDigits=@n90, UVs, DebugTextLocation RG_UpperRight BA_LowerLeft, Component Spacing) {materialFunction=/Engine/Functions/Engine_MaterialFunctions02/Utility/DebugFloat3Values}.GreyScaleOutput
    B=MaterialFunctionCall(Vector3=@n91.Result, MaximumNumberOfDigits=@n90, UVs=@n92) {materialFunction=&DebugFloat3Values}.ColorCodedOutput
   )
   False=ComponentMask(None=@n96.PS) {b=true, g=true, r=true}
  ) {
   desc="To use this feature, Include Hidden Debug Plane in Geometry must be turned on in Houdini and this Material Function's Color RGB (PS) output must be connected to Base Color (at least during debugging). In the debug mode, the data embedded in the static mesh are displayed on a plane (facing local-space +Y).\r\n\r\nThe greyscale XYZ values are:  \r\n\r\ntexture format / position texture count / animation UV row count higher part (0.5-0.99 = HDR; 0.0-0.49 = non-HDR; 0.0-0.24 or 0.5-0.74 = one texture; 0.25-0.49 or 0.75-0.99 = two textures);\r\n\r\nmax pscale (or a custom attribute) in Position Texture Alpha;\r\n\r\nmax RPF (revolutions per frame) found in the animation.\r\n\r\nThe RGB values are: \r\n\r\npowers of two mode active pixels ratio X;\r\n\r\npowers of two mode active pixels ratio Y;\r\n\r\nanimation frame count.\r\n\r\n(Due to some numerical errors in the Unreal debug nodes, occasionally when an embedded value is actually 1 or 0, the plane may misleadingly display 0 or 1 . The does not affect the actual animation and you can safely ignore it. You can verify this is indeed a display bug by temporarily add or remove a few zeros in Embedded Data Rounding.)\r\n\r\nWhen not in the debug mode, the debug plane is invisible and incurs next to no cost."
   group=Houdini VAT - Debug
   parameterName=Debug Embedded Data
   sortPriority=3
  }
  FunctionOutput(None=@n97) {
   description="Outputs Color Texture RGB on the pixel shader (PS). The data can come from the \"Cd\" attribute or a custom vector attribute. The data can be automatically interpolated between frames based on the Material Instance settings. To process this frame's data on the vertex shader, use Color RGBA This Frame (VS). To get interpolated data on the vertex shader, use Color RGBA This Frame (VS), Color RGBA Next Frame (VS), and Interframe Interpolation Alpha (VS) to create your own Lerp."
   outputName="Color RGB (PS)"
  }
  n98 = ComponentMask(None=TextureCoordinate() {coordinateIndex=1}) {desc="Boundary data triangles and debug\r\nplane have a constant V of 1.0,\r\nso they can be easily masked out.", g=true}
  n99 = Constant() {r=0.9}
  n100 = PreSkinnedPosition()
  n101 = Subtract(A=Constant3Vector(), B=@n100) {desc=Invisible vertices LPO}
  n102 = TextureCoordinate()
  n103 = ComponentMask(None=@n102) {r=true}
  n104 = ComponentMask(None=@n102) {g=true}
  n105 = Subtract(
   A=MaterialFunctionCall(X=ConstantBiasScale(None=@n103) {bias=-0.5, scale=500}, Z=ConstantBiasScale(None=@n104) {bias=-1, scale=-500}) {materialFunction=&MakeFloat3}.Result
   B=@n100
  ) {desc=Debug plane LPO}
  n106 = TextureObjectParameter() {
   desc="Describes piece rotations. The RGB channels store the 3 shortest components of a unit quaternion. The longest component is derived at runtime. The A channel stores the normalized RPFs (revolutions per frame). The original RPFs are derived when the A channel data are multiplied by the max RPF from the embedded data. See the tooltip of Debug Embedded Data."
   group=Houdini VAT
   parameterName=Rotation Texture
   samplerType=SAMPLERTYPE_LinearColor
   sortPriority=12
   texture=&Black_1x1_EXR_Texture
  }
  n107 = TextureSample(UVs=@n73, Tex=@n106) {texture=&Black}
  n108 = If(A=@n29, B=Constant() {r=0.5}, "A > B"=@n107.RGBA, "A == B"=@n107.RGBA, "A < B"=ConstantBiasScale(None=@n107.RGBA) {bias=-0.5, scale=2}) {constB=0.5}
  n109 = ComponentMask(None=@n108) {a=true}
  n110 = Abs(None=@n109)
  n111 = If(A=@n29, B=Constant() {r=0.5}, "A > B"=@n110, "A == B"=@n110, "A < B"=Multiply(A=@n110, B=@n88) {constB=10}) {constB=0.5}
  n112 = Multiply(A=Frac(None=@n111)) {constB=0.5, desc=Quaternion 4D angle}
  n113 = Multiply(A=Frac(None=Multiply(A=@n111, B=@n78) {desc="B = Alpha"})) {constB=0.5, desc=Instant 4D angle}
  MaterialExpressionComponentMask_18 = ComponentMask(None=@n108) {b=true, g=true, r=true}
  n114 = Multiply(A=@n74.A) {constB=4}
  MaterialExpressionFloor_1 = Floor(None=@n114)
  n115 = Custom!(XYZ=@MaterialExpressionComponentMask_18, MaxComponent=@MaterialExpressionFloor_1) {
   code="float w = sqrt(1.0 - pow(XYZ.x, 2) - pow(XYZ.y, 2) - pow(XYZ.z, 2));\r\n\r\nswitch(MaxComponent)\r\n{\r\n    case 0:\r\n        return float4(XYZ.x, XYZ.y, XYZ.z, w);\r\n    case 1:\r\n        return float4(-w, XYZ.y, XYZ.z, -XYZ.x);\r\n    case 2:\r\n        return float4(XYZ.x, XYZ.y, -w, -XYZ.z);\r\n    case 3:\r\n        return float4(XYZ.x, -w, XYZ.z, -XYZ.y);\r\n    default:\r\n        return float4(XYZ.x, XYZ.y, XYZ.z, w);\r\n}"
   description=Decode Quaternion
   outputType=CMOT_Float4
   inputs [
    {inputName=XYZ, input {expression=&self:MaterialExpressionComponentMask_18}}
    {inputName=MaxComponent, input {expression=&self:MaterialExpressionFloor_1}}
   ]
  }
  n116 = Multiply(A=Sine(None=Subtract(A=@n112, B=@n113)), B=@n115)
  n117 = TextureSample(UVs=@n50, Tex=@n106) {texture=&Black}
  MaterialExpressionIf_10 = If(A=@n29, B=Constant() {r=0.5}, "A > B"=@n117.RGB, "A == B"=@n117.RGB, "A < B"=ConstantBiasScale(None=@n117.RGB) {bias=-0.5, scale=2}) {constB=0.5}
  n118 = Multiply(A=@n55.A) {constB=4}
  MaterialExpressionFloor_2 = Floor(None=@n118)
  n119 = Custom(XYZ=@MaterialExpressionIf_10, MaxComponent=@MaterialExpressionFloor_2) {
   code="float w = sqrt(1.0 - pow(XYZ.x, 2) - pow(XYZ.y, 2) - pow(XYZ.z, 2));\r\n\r\nswitch(MaxComponent)\r\n{\r\n    case 0:\r\n        return float4(XYZ.x, XYZ.y, XYZ.z, w);\r\n    case 1:\r\n        return float4(-w, XYZ.y, XYZ.z, -XYZ.x);\r\n    case 2:\r\n        return float4(XYZ.x, XYZ.y, -w, -XYZ.z);\r\n    case 3:\r\n        return float4(XYZ.x, -w, XYZ.z, -XYZ.y);\r\n    default:\r\n        return float4(XYZ.x, XYZ.y, XYZ.z, w);\r\n}"
   description=Decode Quaternion
   outputType=CMOT_Float4
   inputs [
    {inputName=XYZ, input {expression=&self:MaterialExpressionIf_10}}
    {inputName=MaxComponent, input {expression=&self:MaterialExpressionFloor_2}}
   ]
  }
  n120 = Multiply(A=Sine(None=@n113), B=Multiply(A=@n119, B=Sign(None=@n109)))
  n121 = Normalize(VectorInput=Divide(A=Add(A=@n116, B=@n120), B=Sine(None=@n112))) {desc=Quaternion slerp result}
  n122 = StaticSwitchParameter(True=If(A=@n109, B=Constant(), "A > B"=@n121, "A == B"=@n115, "A < B"=@n121), False=@n115) {
   desc=Interpolates interframe data when the animation frame is a fractional number. This results in smooth visuals even when you slow down the animation or when the frame rate is unstable.
   group=Houdini VAT
   parameterName=Interframe Interpolation
   sortPriority=1
  }
  n123 = ComponentMask(None=@n122) {b=true, desc=Quaternion XYZ, g=true, r=true}
  n124 = PreSkinnedPosition() {desc=Compatible with SM/ISM/HISM}
  n125 = TextureCoordinate() {coordinateIndex=2}
  MaterialExpressionComponentMask_2 = ComponentMask(None=@n125) {g=true}
  n126 = TextureCoordinate() {coordinateIndex=3}
  MaterialExpressionComponentMask_0 = ComponentMask(None=@n126) {g=true}
  n127 = Abs(None=@MaterialExpressionComponentMask_0)
  n128 = ComponentMask(None=@n125) {r=true}
  n129 = Custom!(C=@MaterialExpressionComponentMask_2, D=@MaterialExpressionComponentMask_0) {
   code="uint c = asuint(C);\r\nuint d = asuint(D);\r\n\r\nc = ((c >> 16) && 0x8000) |\r\n       ((((c >> 23) && 0xFF) - 112) << 10) |\r\n       ((c >> 13) && 0x3FF);\r\n\r\nd = ((d >> 16) && 0x8000) |\r\n       ((((d >> 23) && 0xFF) - 112) << 10) |\r\n       ((d >> 13) && 0x3FF);\r\n\r\nc = (c && 0x8000) | ((c << 2) && 0x7FF8);\r\nd = (d && 0x8000) | ((d << 2) && 0x7FF8);\r\nc = (c << 16) | (d << 3);\r\n\r\nreturn asfloat(c);"
   description=Decode Up
   outputType=CMOT_Float1
   inputs [
    {inputName=C, input {expression=&self:MaterialExpressionComponentMask_2}}
    {inputName=D, input {expression=&self:MaterialExpressionComponentMask_0}}
   ]
  }
  n130 = ComponentMask(None=@n126) {r=true}
  n131 = OneMinus(None=Add(A=Multiply(A=@n128, B=@n128), B=Multiply(A=@n130, B=@n130)))
  n132 = Custom!(A=@MaterialExpressionComponentMask_2, B=@MaterialExpressionComponentMask_0) {
   code="uint a = asuint(A);\r\nuint b = asuint(B);\r\n\r\na = ((a >> 16) && 0x8000) |\r\n       ((((a >> 23) && 0xFF) - 112) << 10) |\r\n       ((a >> 13) && 0x3FF);\r\n\r\nb = ((b >> 16) && 0x8000) |\r\n       ((((b >> 23) && 0xFF) - 112) << 10) |\r\n       ((b >> 13) && 0x3FF);\r\n\r\na = (a && 0x8000) | ((a << 2) && 0x7FF8);\r\nb = (b && 0x8000) | ((b << 2) && 0x7FF8);\r\na = (a << 16) | (b << 3);\r\n\r\nreturn asfloat(a);"
   description=Decode Magnitude
   outputType=CMOT_Float1
   inputs [
    {inputName=A, input {expression=&self:MaterialExpressionComponentMask_2}}
    {inputName=B, input {expression=&self:MaterialExpressionComponentMask_0}}
   ]
  }
  n133 = Multiply(
   A=MaterialFunctionCall(X=@n128, Y=SquareRoot(None=@n131), Z=@n130) {desc=Normalized Direction, materialFunction=&MakeFloat3}.Result
   B=@n132
  ) {desc="16-bit normalized \r\nencoded format"}
  n134 = If(
   A=@n127
   B=Constant() {r=2}
   "A > B"=@n133
   "A == B"=@n133
   "A < B"=MaterialFunctionCall(X=@n128, Y=@n129, Z=@n130) {desc=16-bit original encoded format, materialFunction=&MakeFloat3}.Result
  ) {
   constB=2
   desc="For V of TexCoord[3]:\r\nNormalized format has a 16-bit pattern of *10*** ...\r\nOriginal format has a 16-bit pattern of *01*** ...\r\nOriginal format's abs value is always less than 2.0."
   equalsThreshold=0
  }
  n135 = MaterialFunctionCall(X=@n128, Y=@n130, Z=@MaterialExpressionComponentMask_0) {desc=16/32-bit unencoded format, materialFunction=&MakeFloat3}
  n136 = If(A=Abs(None=@MaterialExpressionComponentMask_2), B=Constant() {r=0.001}, "A > B"=@n134, "A == B"=@n135.Result, "A < B"=@n135.Result) {
   constB=0.001
   desc="16/32-bit unencoded format has 0 in V of TexCoord[2].\r\n16-bit encoded formats have 16-bit patterns of *01*** ...\r\n16-bit encoded formats have a min abs value of 0.0078125."
  }
  n137 = MaterialFunctionCall(Float3=Multiply(A=@n136) {constB=100, desc=Piece pivot local space positions are in Houdini's coordinate system.}) {materialFunction=&BreakOutFloat3Components}
  n138 = MaterialFunctionCall(X=@n137.R, Y=@n137.R, Z=@n137.R) {materialFunction=&MakeFloat3}
  n139 = Subtract(A=@n124, B=@n138.Result) {constB=0, desc=v: vertex offset from pivot}
  n140 = ComponentMask(None=@n122) {a=true, desc=Quaternion W}
  n141 = Add(A=CrossProduct(A=@n123, B=@n139), B=Multiply(A=@n139, B=@n140)) {constB=2}
  n142 = Add(A=Multiply(A=CrossProduct(A=@n123, B=@n141)) {constB=2}, B=@n139) {desc=v_rot}
  n143 = OneMinus(None=Frac(None=@n114))
  n144 = OneMinus(None=Frac(None=@n118))
  n145 = StaticSwitchParameter(True=LinearInterpolate(A=@n143, B=@n144, Alpha=@n78) {desc=Normalized pscale}, False=@n143) {
   desc=Interpolates interframe data when the animation frame is a fractional number. This results in smooth visuals even when you slow down the animation or when the frame rate is unstable.
   group=Houdini VAT
   parameterName=Interframe Interpolation
   sortPriority=1
  }
  n146 = StaticSwitchParameter(True=Multiply(A=@n145, B=@n85) {desc=Original pscale}, False=Constant() {r=1}) {
   defaultValue=true
   desc="The data in Position Texture Alpha usually come from the \"pscale\" attribute. Turn this off if they come from a custom float attribute instead. Doing so instructs the shader to use a constant value of 1.0 for piece scales."
   group=Houdini VAT - Customization
   parameterName=Piece Scales Are in Position Alpha
   sortPriority=0
  }
  n147 = ScalarParameter() {
   defaultValue=1
   desc=This is an override that uniformly scales the size of every piece.
   group=Houdini VAT - Customization
   parameterName=Global Piece Scale Multiplier
   sortPriority=1
  }
  n148 = FunctionInput() {
   bUsePreviewValueAsDefault=true
   description="Provides an additional local-space multiplier to the animated piece scale. This input can vary from piece to piece. To randomize the scale per piece, you can use MF_SimpleRandom or define your own random function. Generate a seed for that function using TexCoord[1]'s RG components, as each piece has a unique pair of UV values in TexCoord[1]."
   inputName="Local-Space Piece Scale Multiplier (VS)"
   sortPriority=1
   previewValue=<1 1 1 1>
  }
  n149 = Multiply(A=@n146, B=Multiply(A=@n147, B=@n148)) {constA=1}
  n150 = Subtract(A=Add(A=Multiply(A=@n142, B=@n149) {desc="× Pscale"}, B=@n82) {desc=+ P}, B=@n124) {desc=Regular LPO}
  n151 = StaticSwitchParameter(
   True=@n150
   False=If(A=Fmod(A=@n46, B=@n40), B=Constant() {r=1}, "A > B"=@n150, "A == B"=Constant(), "A < B"=@n150) {desc=First-frame-dependent WPO}
  ) {
   desc="VAT3.0's new pivot encoding methods will remove most of the unintended cracks in the geometry, unless the piece positions are several metres away from the object origin, or the camera is extremely close to the object. Turn this off to eliminate any remaining seams on the first frame.\r\n\r\nTurn this on if you want to override piece scales or rotations on the first frame."
   group=Houdini VAT - Customization
   parameterName=Animate First Frame
   sortPriority=2
  }
  n152 = FunctionInput() {
   bUsePreviewValueAsDefault=true
   description="Provides an additional local-space offset to the animated vertex position. This input can vary from piece to piece or even vertex to vertex. To randomize the offset per piece, you can use MF_SimpleRandom or define your own random function. Generate a seed for that function using TexCoord[1]'s RG components, as each piece has a unique pair of UV values in TexCoord[1]."
   inputName="Additional Local Position Offset (VS)"
   sortPriority=2
  }
  n153 = StaticSwitchParameter(
   True=If(A=@n98, B=@n99, "A > B"=@n105, "A == B"=@n105, "A < B"=@n101)
   False=If(A=@n98, B=@n99, "A > B"=@n101, "A == B"=@n101, "A < B"=Add(A=@n151, B=@n152))
  ) {
   desc="To use this feature, Include Hidden Debug Plane in Geometry must be turned on in Houdini and this Material Function's Color RGB (PS) output must be connected to Base Color (at least during debugging). In the debug mode, the data embedded in the static mesh are displayed on a plane (facing local-space +Y).\r\n\r\nThe greyscale XYZ values are:  \r\n\r\ntexture format / position texture count / animation UV row count higher part (0.5-0.99 = HDR; 0.0-0.49 = non-HDR; 0.0-0.24 or 0.5-0.74 = one texture; 0.25-0.49 or 0.75-0.99 = two textures);\r\n\r\nmax pscale (or a custom attribute) in Position Texture Alpha;\r\n\r\nmax RPF (revolutions per frame) found in the animation.\r\n\r\nThe RGB values are: \r\n\r\npowers of two mode active pixels ratio X;\r\n\r\npowers of two mode active pixels ratio Y;\r\n\r\nanimation frame count.\r\n\r\n(Due to some numerical errors in the Unreal debug nodes, occasionally when an embedded value is actually 1 or 0, the plane may misleadingly display 0 or 1 . The does not affect the actual animation and you can safely ignore it. You can verify this is indeed a display bug by temporarily add or remove a few zeros in Embedded Data Rounding.)\r\n\r\nWhen not in the debug mode, the debug plane is invisible and incurs next to no cost."
   group=Houdini VAT - Debug
   parameterName=Debug Embedded Data
   sortPriority=3
  }
  FunctionOutput(None=Transform(None=@n153) {transformSourceType=TRANSFORMSOURCE_Local}) {description=Connect this to World Position Offset., outputName=World Position Offset, sortPriority=4}
  n154 = Constant3Vector()
  n155 = FunctionInput() {
   bUsePreviewValueAsDefault=true
   description="The RGB output of a regular surface normal map or the combined RGB output of multiple blended surface normal maps. Do not input any world-space normal here. Instead, blend any world-space normal with this Material Function's Normal output."
   inputName="Surface Normal Maps RGB (PS)"
   previewValue=<0 0 1 1>
  }
  n156 = AppendVector(A=ComponentMask(None=TextureCoordinate() {coordinateIndex=3}) {g=true}, B=TextureCoordinate() {coordinateIndex=4}) {desc=Tangent WS}
  n157 = AppendVector(A=TextureCoordinate() {coordinateIndex=2}, B=ComponentMask(None=TextureCoordinate() {coordinateIndex=3}) {r=true}) {desc=Normal WS}
  n158 = StaticSwitchParameter(
   True=MaterialFunctionCall(VectorToTransform=@n155, BasisX=@n156, BasisY=CrossProduct(A=@n157, B=@n156), BasisZ=@n157) {materialFunction=&Transform3x3Matrix}.Transformed Vector
   False=@n157
  ) {
   defaultValue=true
   desc="Turn this on if you need to apply surface normal maps. When importing the FBX, set Normal Import Method to \"Import Normals and Tangents\"."
   group=Houdini VAT
   parameterName=Support Surface Normal Maps
   sortPriority=4
  }
  n159 = StaticSwitchParameter(True=@n154, False=@n158) {
   desc="To use this feature, Include Hidden Debug Plane in Geometry must be turned on in Houdini and this Material Function's Color RGB (PS) output must be connected to Base Color (at least during debugging). In the debug mode, the data embedded in the static mesh are displayed on a plane (facing local-space +Y).\r\n\r\nThe greyscale XYZ values are:  \r\n\r\ntexture format / position texture count / animation UV row count higher part (0.5-0.99 = HDR; 0.0-0.49 = non-HDR; 0.0-0.24 or 0.5-0.74 = one texture; 0.25-0.49 or 0.75-0.99 = two textures);\r\n\r\nmax pscale (or a custom attribute) in Position Texture Alpha;\r\n\r\nmax RPF (revolutions per frame) found in the animation.\r\n\r\nThe RGB values are: \r\n\r\npowers of two mode active pixels ratio X;\r\n\r\npowers of two mode active pixels ratio Y;\r\n\r\nanimation frame count.\r\n\r\n(Due to some numerical errors in the Unreal debug nodes, occasionally when an embedded value is actually 1 or 0, the plane may misleadingly display 0 or 1 . The does not affect the actual animation and you can safely ignore it. You can verify this is indeed a display bug by temporarily add or remove a few zeros in Embedded Data Rounding.)\r\n\r\nWhen not in the debug mode, the debug plane is invisible and incurs next to no cost."
   group=Houdini VAT - Debug
   parameterName=Debug Embedded Data
   sortPriority=3
  }
  FunctionOutput(None=@n159) {
   description="Outputs the world-space normal. On Material Details panel, turn off Tangent Space Normal. Connect this to Normal."
   outputName="Normal (Tangent Space Normal Off)"
   sortPriority=3
  }
  FunctionOutput(None=TextureCoordinate() {coordinateIndex=1}) {
   description="You have two options: either connect this to Customized UV1 or remember not to connect anything to Customized UV1. This output's only purpose is to prevent accidental overwritting of UV1 in the Material, which must not be allowed."
   outputName="Customized UV1 (*)"
   sortPriority=21
  }
  n160 = PreSkinnedNormal() {desc=Compatible with SM/ISM/HISM}
  n161 = CrossProduct(A=@n123, B=@n160)
  n162 = Multiply(A=@n160, B=@n140)
  n163 = Multiply(A=CrossProduct(A=@n123, B=Add(A=@n161, B=@n162) {constB=2})) {constB=2}
  n164 = Normalize(VectorInput=Transform(None=Add(A=@n163, B=@n160) {constB=2}) {transformSourceType=TRANSFORMSOURCE_Local})
  FunctionOutput(None=ComponentMask(None=@n164) {desc=Customized UV2, g=true, r=true}) {description=Connect this to Customized UV2., outputName=Customized UV2, sortPriority=22}
  n165 = TextureCoordinate() {coordinateIndex=4}
  n166 = MaterialFunctionCall(
   X=ComponentMask(None=@n165) {r=true}
   Y=ComponentMask(None=TextureCoordinate() {coordinateIndex=5}) {r=true}
   Z=ComponentMask(None=@n165) {g=true}
  ) {
   desc=UE currently does not support Pre-Skinned Local Tangent.
   materialFunction=&MakeFloat3
  }
  n167 = Transform(None=VertexTangentWS()) {transformSourceType=TRANSFORMSOURCE_World, transformType=TRANSFORM_Local}
  n168 = StaticSwitchParameter(True=@n166.Result, False=@n167) {
   defaultValue=true
   desc=Turn this on if you need to apply surface normal maps. Support Real-Time Instancing must be turned on in Houdini for this to work.
   group=Houdini VAT - Instancing
   parameterName=Support Instancing with Surface Normal Maps
   sortPriority=7
  }
  n169 = StaticSwitchParameter(True=@n168, False=@n167) {
   desc="Prior to Vertex Animation Textures ROP 3.0, you needed to copy and paste a list of parameter values from the Houdini VAT node or an exported JSON data table. Those data are now embedded in the bounds of the static mesh and automatically derived at runtime. However, there are certain infrequent limitations: 1. when the Material is used with an instanced mesh, the instance transform may corrupt the embedded data; 2. if you change the static mesh bound scale, the embedded data may also be corrupted; but you should rarely have any reason to change the bound scale, because unlike prior versions, the bounds now encompass all frames of the animation, which prevents the object from being frustum-culled incorrectly when its first frame's bounds are out of view; that being said, overriding the actor/component scale is totally safe."
   group=Houdini VAT - Instancing
   parameterName=Support Legacy Parameters and Instancing
   sortPriority=0
  }
  n170 = Add(A=CrossProduct(A=@n123, B=@n169), B=Multiply(A=@n169, B=@n140)) {constB=2}
  n171 = Add(A=Multiply(A=CrossProduct(A=@n123, B=@n170)) {constB=2}, B=@n169) {constB=2}
  n172 = StaticSwitchParameter(True=Normalize(VectorInput=Transform(None=@n171) {transformSourceType=TRANSFORMSOURCE_Local}), False=Constant3Vector()) {
   defaultValue=true
   desc="Turn this on if you need to apply surface normal maps. When importing the FBX, set Normal Import Method to \"Import Normals and Tangents\"."
   group=Houdini VAT
   parameterName=Support Surface Normal Maps
   sortPriority=4
  }
  FunctionOutput(None=AppendVector(A=ComponentMask(None=@n164) {b=true}, B=ComponentMask(None=@n172) {r=true}) {desc=Customized UV3}) {description=Connect this to Customized UV3., outputName=Customized UV3, sortPriority=23}
  FunctionOutput(None=ComponentMask(None=@n172) {b=true, desc=Customized UV4, g=true}) {description=Connect this to Customized UV4., outputName=Customized UV4, sortPriority=24}
  FunctionOutput(None=Multiply(A=@n143, B=@n85) {desc=Original pscale this frame}) {
   description="Unlike other texture channels, Position Texture Alpha tends to store data in a special way. The data usually go through some form of normalization or compression. This output gives you this frame's data in their originally intended format, not the raw data in Position Texture Alpha. To get interpolated data, use this with Data in Position Alpha Next Frame (VS) and Interframe Interpolation Alpha (VS) to create your own Lerp."
   outputName="Data in Position Alpha This Frame (VS)"
   sortPriority=9
  }
  FunctionOutput(None=Multiply(A=@n144, B=@n85) {desc=Original pscale next frame}) {
   description="Unlike other texture channels, Position Texture Alpha tends to store data in a special way. The data usually go through some form of normalization or compression. This output gives you next frame's data in their originally intended format, not the raw data in Position Texture Alpha. To get interpolated data, use this with Data in Position Alpha This Frame (VS) and Interframe Interpolation Alpha (VS) to create your own Lerp."
   outputName="Data in Position Alpha Next Frame (VS)"
   sortPriority=10
  }
  FunctionOutput(None=@n78) {
   description="Outputs the fractional part of the frame number on the vertex shader (VS). This can be used to interpolate various outputs. All outputs ending with (PS) can be automatically interpolated based on the Material Instance settings. All outputs ending with This Frame (VS) / Next Frame (VS) can be interpolated explicity downstream from this Material Function. This flexibility allows you to decide which sets of data need interpolation and how to implement the interpolation method. It can be linear, spherical linear, or more."
   outputName="Interframe Interpolation Alpha (0-1) (VS)"
   sortPriority=15
  }
  FunctionOutput(None=@n71) {
   description="Outputs this frame's animation progress on the vertex shader (VS). This progress is expressed in the 0-1 range with 0 being the inclusive start of the animation and 1 being the exclusive end. To get interpolated data, use this with Animation Progress Next Frame (0-1) (VS) and Interframe Interpolation Alpha (VS) to create your own Lerp."
   outputName="Animation Progress This Frame (0-1) (VS)"
   sortPriority=16
  }
  n173 = StaticSwitchParameter(True=Constant() {r=1}, False=ComponentMask(None=@n96.PS) {a=true}) {
   desc="To use this feature, Include Hidden Debug Plane in Geometry must be turned on in Houdini and this Material Function's Color RGB (PS) output must be connected to Base Color (at least during debugging). In the debug mode, the data embedded in the static mesh are displayed on a plane (facing local-space +Y).\r\n\r\nThe greyscale XYZ values are:  \r\n\r\ntexture format / position texture count / animation UV row count higher part (0.5-0.99 = HDR; 0.0-0.49 = non-HDR; 0.0-0.24 or 0.5-0.74 = one texture; 0.25-0.49 or 0.75-0.99 = two textures);\r\n\r\nmax pscale (or a custom attribute) in Position Texture Alpha;\r\n\r\nmax RPF (revolutions per frame) found in the animation.\r\n\r\nThe RGB values are: \r\n\r\npowers of two mode active pixels ratio X;\r\n\r\npowers of two mode active pixels ratio Y;\r\n\r\nanimation frame count.\r\n\r\n(Due to some numerical errors in the Unreal debug nodes, occasionally when an embedded value is actually 1 or 0, the plane may misleadingly display 0 or 1 . The does not affect the actual animation and you can safely ignore it. You can verify this is indeed a display bug by temporarily add or remove a few zeros in Embedded Data Rounding.)\r\n\r\nWhen not in the debug mode, the debug plane is invisible and incurs next to no cost."
   group=Houdini VAT - Debug
   parameterName=Debug Embedded Data
   sortPriority=3
  }
  FunctionOutput(None=@n173) {
   description="Outputs Color Texture Alpha on the pixel shader (PS). The data can come from the \"Alpha\" attribute or a custom float attribute. The data can be automatically interpolated between frames based on the Material Instance settings. To process this frame's data on the vertex shader, use Color RGBA This Frame (VS). To get interpolated data on the vertex shader, use Color RGBA This Frame (VS), Color RGBA Next Frame (VS), and Interframe Interpolation Alpha (VS) to create your own Lerp."
   outputName="Color Alpha (PS)"
   sortPriority=1
  }
  n174 = TextureSample(UVs=@n73, Tex=@n53) {texture=&Black}
  n175 = StaticSwitchParameter(True=LinearInterpolate(A=@n174.RGBA, B=@n54.RGBA, Alpha=@n78), False=@n174.RGBA) {
   desc="There is a cost associated with interpolating each vertex animation texture. While Interframe Interpolation is the overall switch, you can independently decide whether or not to interpolate Spare Color Texture. This only affects the pixel shader output Spare Color RGBA (PS). The outputs ending with This Frame (VS) / Next Frame (VS) are never interpolated."
   group=Houdini VAT
   parameterName=Interpolate Spare Color
   sortPriority=3
  }
  n176 = StaticSwitchParameter(True=@n175, False=@n174.RGBA) {
   desc=Interpolates interframe data when the animation frame is a fractional number. This results in smooth visuals even when you slow down the animation or when the frame rate is unstable.
   group=Houdini VAT
   parameterName=Interframe Interpolation
   sortPriority=2
  }
  n177 = StaticSwitchParameter(True=@n154, False=VertexInterpolator(VS=@n176).PS) {
   desc="To use this feature, Include Hidden Debug Plane in Geometry must be turned on in Houdini and this Material Function's Color RGB (PS) output must be connected to Base Color (at least during debugging). In the debug mode, the data embedded in the static mesh are displayed on a plane (facing local-space +Y).\r\n\r\nThe greyscale XYZ values are:  \r\n\r\ntexture format / position texture count / animation UV row count higher part (0.5-0.99 = HDR; 0.0-0.49 = non-HDR; 0.0-0.24 or 0.5-0.74 = one texture; 0.25-0.49 or 0.75-0.99 = two textures);\r\n\r\nmax pscale (or a custom attribute) in Position Texture Alpha;\r\n\r\nmax RPF (revolutions per frame) found in the animation.\r\n\r\nThe RGB values are: \r\n\r\npowers of two mode active pixels ratio X;\r\n\r\npowers of two mode active pixels ratio Y;\r\n\r\nanimation frame count.\r\n\r\n(Due to some numerical errors in the Unreal debug nodes, occasionally when an embedded value is actually 1 or 0, the plane may misleadingly display 0 or 1 . The does not affect the actual animation and you can safely ignore it. You can verify this is indeed a display bug by temporarily add or remove a few zeros in Embedded Data Rounding.)\r\n\r\nWhen not in the debug mode, the debug plane is invisible and incurs next to no cost."
   group=Houdini VAT - Debug
   parameterName=Debug Embedded Data
   sortPriority=3
  }
  FunctionOutput(None=@n177) {
   description="Outputs Spare Color Texture RGBA on the pixel shader (PS). The data strictly come from the \"Cd2\" and \"Alpha2\" attributes. You can use Spare Color Texture to store the actual colors and Color Texture to store custom attributes. The data can be automatically interpolated between frames based on the Material Instance settings. To process this frame's data on the vertex shader, use Spare Color RGBA This Frame (VS). To get interpolated data on the vertex shader, use Spare Color RGBA This Frame (VS), Spare Color RGBA Next Frame (VS), and Interframe Interpolation Alpha (VS) to create your own Lerp."
   outputName="Spare Color RGBA (PS)"
   sortPriority=2
  }
  FunctionOutput(None=@n48) {
   description="Outputs next frame's animation progress on the vertex shader (VS). This progress is expressed in the 0-1 range with 0 being the inclusive start of the animation and 1 being the exclusive end. To get interpolated data, use this with Animation Progress This Frame (0-1) (VS) and Interframe Interpolation Alpha (VS) to create your own Lerp."
   outputName="Animation Progress Next Frame (0-1) (VS)"
   sortPriority=17
  }
  FunctionOutput(None=TextureCoordinate() {coordinateIndex=5}) {
   description="This output is only relevant if you are using this Material with ISM/HISM, in which case you have two options: either connect this to Customized UV5 or remember not to connect anything to Customized UV5. This output's only purpose is to prevent accidental overwritting of UV5 in the Material, which must not be allowed when the Material is used with ISM/HISM."
   outputName="Customized UV5 (ISM/HISM Only) (*)"
   sortPriority=25
  }
  FunctionOutput(None=@n49) {
   description="On the vertex shader (VS), outputs next frame's V (as in UVs) used to sample the vertex animation textures. To get interpolated values, use this with Sampling V This Frame (0-1) (VS) and Interframe Interpolation Alpha (VS) to create your own Lerp."
   outputName="Sampling V Next Frame (0-1) (VS)"
   sortPriority=6
  }
  FunctionOutput(None=@n72) {
   description="On the vertex shader (VS), outputs this frame's V (as in UVs) used to sample the vertex animation textures. To get interpolated values, use this with Sampling V Next Frame (0-1) (VS) and Interframe Interpolation Alpha (VS) to create your own Lerp."
   outputName="Sampling V This Frame (0-1) (VS)"
   sortPriority=5
  }
  FunctionOutput(None=@n93.RGBA) {
   description="Outputs this frame's Color Texture RGBA on the vertex shader (VS). The data can come from the \"Cd\" and \"Alpha\" attributes, or custom vector and float attributes. To get interpolated data, use this with Color RGBA Next Frame (VS) and Interframe Interpolation Alpha (VS) to create your own Lerp."
   outputName="Color RGBA This Frame (VS)"
   sortPriority=11
  }
  FunctionOutput(None=@n174.RGBA) {
   description="Outputs this frame's Spare Color Texture RGBA on the vertex shader (VS). The data strictly come from the \"Cd2\" and \"Alpha2\" attributes. You can use Spare Color Texture to store the actual colors and Color Texture to store custom attributes. To get interpolated data, use this with Spare Color RGBA Next Frame (VS) and Interframe Interpolation Alpha (VS) to create your own Lerp."
   outputName="Spare Color RGBA This Frame (VS)"
   sortPriority=13
  }
  FunctionOutput(None=@MaterialExpressionIf_7) {
   description="Outputs this frame's piece local position on the vertex shader (VS). If Position Texture is non-HDR, this output is Position Texture RGB transformed by the object local bounds. To get interpolated data, use this with the outputs Piece Local Position Next Frame (VS) and Interframe Interpolation Alpha (VS) to create your own Lerp."
   outputName="Piece Local Position This Frame (VS)"
   sortPriority=7
  }
  FunctionOutput(None=@n138.Result) {
   description="Outputs the piece rest frame local position on the vertex shader (VS). This can be used to make different pieces behave differently based on where they are before the animation starts."
   outputName="Piece Rest Frame Local Position (VS)"
   sortPriority=18
  }
  TransformPosition(None=WorldPosition() {worldPositionShaderOffset=WPT_ExcludeAllShaderOffsets}.XYZ) {transformSourceType=TRANSFORMPOSSOURCE_World}
  Transform(None=VertexNormalWS()) {transformSourceType=TRANSFORMSOURCE_World, transformType=TRANSFORM_Local}
  n178 = Constant3Vector() {constant=<rgba 0 0.4167 1 0>}
  VertexInterpolator(VS=If(A=@n127, B=Constant() {r=2}, "A > B"=@n178, "A == B"=@n178, "A < B"=Constant3Vector() {constant=<rgba 1 0.0833 0 0>}) {constB=2}) {desc="Debug Pivot Accuracy: \"Very High\""}
 ]
}
