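# Masked, lit Niagara mesh-particle material for the Floret Crumble emote effect, using rigid-body vertex animation textures (VAT) and Endear prop textures.
# The particle dynamic parameter Time sets the displayed frame in the VAT rigid-body function.
# A height-based erosion gradient compares pixel Z against particle Z, offset and sized by the ErosionStartHeight and ErosionLength dynamic parameters.
# A SmoothStep combines that gradient with packed fire noise to give a hard, broken-up erosion edge.
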Material /Game/Effects/Fort_Effects/Emotes/Floret/Crumble/M_Emote_Floret_Crumble_RBDVAT
graph {
 properties {
  bTangentSpaceNormal=false
  bUsedWithNiagaraMeshParticles=true
  blendMode=BLEND_Masked
  numCustomizedUVs=5
  pixelDepthOffsetMode=PDOM_Legacy
  translucencyLightingMode=TLM_VolumetricDirectional
  usedShadingModels=Default Lit
 }
 nodes [
  n1 = TransformPosition(None=ParticlePositionWS()) {transformSourceType=TRANSFORMPOSSOURCE_World}
  n2 = DynamicParameter() {paramNames [Time, ErrosionStartHeight, ErrosionLength, Param4]}
  Add(A=@n1, B=@n2.ErrosionStartHeight)
  Constant3Vector() {constant=<rgba 0.4072 0.0242 0.0704 1>}
  Constant() {r=0.8}
  Constant() {r=0.3}
  MaterialFunctionCall!(
   "Time Override (VS)"
   "Display Frame Override (VS)"=@n2.Time
   "Surface Normal Maps RGB (PS)"
   "Local-Space Piece Scale Multiplier (VS)"
   "Additional Local Position Offset (VS)"
   "Local-Space Piece Acceleration This Frame (VS)"
  ) {materialFunction=/Game/Effects/Niagara/MasterMaterials/VAT/MF_VAT_RigidBodyDynamics}
  MaterialFunctionCall!(Input, Input Low, Input High) {materialFunction=/Engine/Functions/Engine_MaterialFunctions03/Math/RemapValueRangeNormalized}
  MaterialFunctionCall() {materialFunction=&MakeFloat3}
  ParticleColor()
  ScalarParameter() {parameterName=UpperLimit}
  ScalarParameter() {defaultValue=50, parameterName=BandLength}
  n3 = TransformPosition(None=WorldPosition().XYZ) {transformSourceType=TRANSFORMPOSSOURCE_World}
  n4 = DynamicParameter() {paramNames [Time, ErrosionStartHeight, ErrosionLength, Param4]}
  n5 = Add(
   A=MaterialFunctionCall(Float3=@n1) {materialFunction=&BreakOutFloat3Components}.B
   B=@n4.ErrosionStartHeight
  ) {constB=250}
  n6 = TextureSample(UVs=TextureCoordinate()) {texture=/Game/Effects/Fort_Effects/Textures/Fire/HF_FireNoisesPacked}
  SmoothStep(
   Min=MaterialFunctionCall(
    Input=MaterialFunctionCall(Float3=@n3) {materialFunction=&BreakOutFloat3Components}.B
    Input Low=@n5
    Input High=Subtract(A=@n5, B=@n4.ErrosionLength)
   ) {materialFunction=&RemapValueRangeNormalized}.Result
   Value=Clamp(None=Add(A=@n6.R) {constB=0.5}) {maxDefault=0.999}
  ) {constMin=0.99}
  TextureSample() {texture=/Game/Animation/Props/Gadgets_Endear/Textures/T_Emote_Prop_Endear_D}
  TextureSample() {samplerType=SAMPLERTYPE_Masks, texture=/Game/Animation/Props/Gadgets_Endear/Textures/T_Emote_Prop_Endear_S}
 ]
}
